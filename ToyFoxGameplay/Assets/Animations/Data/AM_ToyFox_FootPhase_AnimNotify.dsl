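; Animation modifier that generates ToyFox foot plant and foot pass notifies, plus a foot phase curve, from an animation sequence's sync markers.
; On apply, it adds each enabled track, then maps L, R, LeftPass and RightPass markers to left or right notifies at the marker times.
; A helper converts marker names to a phase enum, keying the curve to 1 at plants and 0 at passes.
; Revert removes both notify tracks but leaves the curve.
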(event EventOnApply
  (params (AnimationSequence :type Anim%20Sequence%20Object%20Reference))
  (Variables|Default|SetAnimation%20Sequence AnimationSequence)
  (if (Variables|Default|GetPlot_FootPlant_Notifies)
    (AnimationBlueprintLibrary|AnimationNotifies|AddAnimationNotifyTrack (Variables|Default|GetAnimation%20Sequence) (Variables|Default|GetAN_Plant_TrackName) "(R=0.001982,G=1.000000,B=0.001754,A=1.000000)"))
  (if (Variables|Default|GetPlot_FootPass_Notifies)
    (AnimationBlueprintLibrary|AnimationNotifies|AddAnimationNotifyTrack (Variables|Default|GetAnimation%20Sequence) (Variables|Default|GetAN_Pass_TrackName) "(R=0.000000,G=1.000000,B=0.964710,A=1.000000)"))
  (if (Variables|Default|GetPlot_FootPhaseCurveTrack)
    (AnimationBlueprintLibrary|Curves|AddCurve (Variables|Default|GetAnimation%20Sequence) (Variables|Default|GetFootPhaseCurve_TrackName)))
  (for _array_element (AnimationBlueprintLibrary|MarkerSyncing|GetAnimationSyncMarkers (Variables|Default|GetAnimation%20Sequence))
    (CallFunction|SetFootPhaseEnum (Utilities|Struct|BreakAnimSyncMarker _array_element))
    (Utilities|FlowControl|Sequence
      (:then_0
        (if (and (or (== (Utilities|Struct|BreakAnimSyncMarker _array_element) "L") (== (Utilities|Struct|BreakAnimSyncMarker _array_element) "R")) (Variables|Default|GetPlot_FootPlant_Notifies))
          (AnimationBlueprintLibrary|NotifyEvents|AddAnimationNotifyEvent (Variables|Default|GetAnimation%20Sequence) (Variables|Default|GetAN_Plant_TrackName) (.Time (Utilities|Struct|BreakAnimSyncMarker _array_element)) (select (== (Utilities|Struct|BreakAnimSyncMarker _array_element) "L") "/ToyFoxGameplay/Assets/Animations/Data/AN_ToyFox_LFoot_Plant.AN_ToyFox_LFoot_Plant_C" "/ToyFoxGameplay/Assets/Animations/Data/AN_ToyFox_RFoot_Plant.AN_ToyFox_RFoot_Plant_C"))))
      (:then_1
        (if (and (or (== (Utilities|Struct|BreakAnimSyncMarker _array_element) "LeftPass") (== (Utilities|Struct|BreakAnimSyncMarker _array_element) "RightPass")) (Variables|Default|GetPlot_FootPass_Notifies))
          (AnimationBlueprintLibrary|NotifyEvents|AddAnimationNotifyEvent (Variables|Default|GetAnimation%20Sequence) (Variables|Default|GetAN_Pass_TrackName) (.Time (Utilities|Struct|BreakAnimSyncMarker _array_element)) (select (== (Utilities|Struct|BreakAnimSyncMarker _array_element) "LeftPass") "/ToyFoxGameplay/Assets/Animations/Data/AN_ToyFox_LFoot_Pass.AN_ToyFox_LFoot_Pass_C" "/ToyFoxGameplay/Assets/Animations/Data/AN_ToyFox_RFoot_Pass.AN_ToyFox_RFoot_Pass_C"))))
      (:then_2
        (if (Variables|Default|GetPlot_FootPhaseCurveTrack)
          (AnimationBlueprintLibrary|Curves|AddFloatCurveKey (Variables|Default|GetAnimation%20Sequence) (Variables|Default|GetFootPhaseCurve_TrackName) (.Time (Utilities|Struct|BreakAnimSyncMarker _array_element)) (Utilities|Select :Index (Variables|Default|GetE_FootPhase) :NewEnumerator0 0.0 :NewEnumerator1 1.0 :NewEnumerator2 0.0 :NewEnumerator3 1.0)))))))

(event EventOnRevert
  (params (AnimationSequence :type Anim%20Sequence%20Object%20Reference))
  (AnimationBlueprintLibrary|AnimationNotifies|RemoveAnimationNotifyTrack AnimationSequence (Variables|Default|GetAN_Plant_TrackName))
  (AnimationBlueprintLibrary|AnimationNotifies|RemoveAnimationNotifyTrack AnimationSequence (Variables|Default|GetAN_Pass_TrackName)))

(fn SetFootPhaseEnum
  (params (MarkerName :type Name))
  (Utilities|FlowControl|Sequence
    (:then_0
      (if (== MarkerName "L")
        (Variables|Default|SetE_FootPhase "NewEnumerator1")))
    (:then_1
      (if (== MarkerName "R")
        (Variables|Default|SetE_FootPhase "NewEnumerator3")))
    (:then_2
      (if (== MarkerName "LeftPass")
        (Variables|Default|SetE_FootPhase "NewEnumerator0")))
    (:then_3
      (if (== MarkerName "RightPass")
        (Variables|Default|SetE_FootPhase "NewEnumerator2")))
    (:then_4
      _))
  (return))
