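; Anim notify for the ToyFox giant pawn that signals a left foot plant during animation playback.
; On notify, it gets the mesh's anim instance and casts it to ToyFoxGiantPawnAnimBP.
; It then calls UpdateFootPhase with an enumerator value and the LPlant tag, leaving foot phase handling to the anim blueprint.
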(fn Received_Notify
  (params (MeshComp :type Skeletal%20Mesh%20Component%20Object%20Reference) (Animation :type Anim%20Sequence%20Base%20Object%20Reference) (EventReference :type Anim%20Notify%20Event%20Reference%20Structure%20%28by%20ref%29))
  (bind _astoy_fox_giant_pawn_anim_bp (Utilities|Casting|CastToToyFoxGiantPawnAnimBP (Components|SkeletalMesh|GetAnimInstance MeshComp)))
  (Class|ToyFoxGiantPawnAnimBP|UpdateFootPhase _astoy_fox_giant_pawn_anim_bp "NewEnumerator1" "LPlant")
  (return false))
